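# Editor utility widget that mirrors selected level actors, intended as building pieces, across a center point on chosen world axes.
# The center is either the selection's average location snapped to the Fort building grid or a set world point.
# FlipActors moves each actor inside an undo transaction, yawing it 180 degrees when movement aligns with its right vector.
# A details view exposes the settings, which persist through the base widget's save game arrays.
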EditorUtilityWidgetBlueprint /Game/NeverCook/EditorUtilities/BuildingTools/BuildingFlipper/EUW_BuildingFlipper
graphs [
 EventGraph {nodes [
   "On Clicked (TA_Generic01_Button)"!() -> (OutputDelegate, then, widget:User Widget Object Reference, event:Pointer Event Structure) {is=title K2Node_ComponentBoundEvent}
   > n1 = GetSelectedLevelActors(self=Get EditorActorSubsystem())
   > n2 = GetActorArrayAverageLocation!(execute, Actors:Array of Actor Object References=@n1) -> (then, ReturnValue:Vector)
   > FlipActors!(
    execute
    self
    actors:Array of Actor Object References=@n1
    centerPoint:Vector=Select(
     Option 0:Vector=Get WorldCenterIfNotUsingSelection() -> (WorldCenterIfNotUsingSelection:Vector)
     Option 1:Vector=SnapToFortBuildingGrid!(InVector:"Vector (by ref)"=@n2, PreSnapOffset:Vector="0, 0, 0", PostSnapOffset:Vector="0, 0, 0") -> (ReturnValue:Vector)
     Index=Get UseSelectionCenter() -> (UseSelectionCenter:bool)
    ) -> (ReturnValue:Vector)
    invertVector:Vector=Get WorldInvertVector() -> (WorldInvertVector:Vector)
   ) -> (then)
   "On Property Changed (DetailsView_0)"!() -> (OutputDelegate, then, PropertyName:Name=None) {is=title K2Node_ComponentBoundEvent}
   > SaveGame_Save(self:TA Base Editor Widget Object Reference)
   Event Construct()
   > Parent: Construct()
   > SetObject(self=Get DetailsView_0() -> (DetailsView_0:Details View Object Reference), NewObject=Self-Reference())
  ]}
 FlipActors {nodes [
   n1 = Get l_actor() -> (l_actor:Actor Object Reference)
   n2 = vector - vector(A=Get l_newLocation() -> (l_newLocation:Vector), B=Get l_actorLoc() -> (l_actorLoc:Vector))
   n3 = "Absolute (Float)"(A=Dot Product!(
     A:Vector=Normalize!(A:Vector=@n2, Tolerance:float=0.0001) -> (ReturnValue:Vector) {is=title K2Node_CallFunction}
     B:Vector=GetActorRightVector!(self:Actor Object Reference=@n1) -> (ReturnValue:Vector)
    ) -> (ReturnValue:double))
   n4 = PrintString(InString="To String (Float)"(InDouble=@n3))
   Branch(execute=[@n18.then, @n4.then], Condition="float > float"(A=@n3, B=0.8))
   > Add Actor World Rotation!(execute, self:Actor Object Reference=Get l_actor(), DeltaRotation:Rotator="0,180,0", bSweep:bool=false, bTeleport:bool=false) -> (then, SweepHitResult:Hit Result Structure)
   n5 = FlipActors!() -> (then, actors:Array of Actor Object References, centerPoint:Vector=~, invertVector:Vector="-1,0,0") {is=title K2Node_FunctionEntry}
   > Set l_center(l_center:Vector=@n5.centerPoint) -> (Output_Get:Vector)
   n6 = BreakVector(InVec=@n5.invertVector)
   n7 = "Clamp (Integer)"(Value=Round(A="Absolute (Float)"(A=@n6.Y)) {is=title K2Node_CallFunction}, Min=0, Max=1) -> (ReturnValue=0)
   n8 = "Clamp (Integer)"(Value=Round(A="Absolute (Float)"(A=@n6.Z)), Min=0, Max=1) -> (ReturnValue=0)
   n9 = "Clamp (Integer)"(Value=Round(A="Absolute (Float)"(A=@n6)), Min=0, Max=1) -> (ReturnValue=0)
   > Set l_vector(l_vector:Vector=MakeVector(X="To Float (Integer)"(InInt=@n9), Y="To Float (Integer)"(InInt=@n7), Z="To Float (Integer)"(InInt=@n8))) -> (Output_Get:Vector)
   > BeginTransaction(Context=BuildingFlipper, PrimaryObject=Self-Reference())
   > n10 = For Each Loop(Array:Array of Actor Object References=@n5) -> (Array Element:Actor Object Reference)
   EndTransaction(execute=@n10.Completed)
   n11 = Set l_actor(execute=@n10.LoopBody, l_actor:Actor Object Reference=@n10) -> (Output_Get:Actor Object Reference)
   > TransactObject(Object=@n11)
   > Set l_actorLoc(l_actorLoc:Vector=Get Actor Location(self=Get l_actor())) -> (Output_Get:Vector)
   n12 = BreakVector(InVec=Get l_actorLoc())
   n13 = Get l_vector() -> (l_vector:Vector)
   n14 = BreakVector(InVec=@n13)
   n15 = vector - vector(A=Get l_actorLoc(), B=Get l_center() -> (l_center:Vector))
   n16 = BreakVector(InVec=vector - vector(A=Get l_center(), B=vector * vector(A=@n15, B:Vector=@n13) -> (ReturnValue=~)))
   > n17 = Set l_newLocation(l_newLocation:Vector=MakeVector(X=Lerp(A=@n12, B=@n16, Alpha=@n14), Y=Lerp(A=@n12.Y, B=@n16.Y, Alpha=@n14.Y), Z=Lerp(A=@n12.Z, B=@n16.Z, Alpha=@n14.Z))) -> (Output_Get:Vector)
   > n18 = Set Actor Location!(execute, self:Actor Object Reference=Get l_actor(), NewLocation:Vector=@n17, bSweep:bool=false, bTeleport:bool=false) -> (then, SweepHitResult:Hit Result Structure, ReturnValue:bool) {is=title K2Node_CallFunction}
   BreakVector(InVec="0, 0, 0")
   Cast To BuildingActor!(execute, Object:Wildcard) -> (then, CastFailed, AsBuilding Actor:Building Actor Object Reference) {is=title K2Node_DynamicCast}
  ]}
 SaveGame_PostLoad {nodes [
   n1 = SaveGame_PostLoad!() -> (then, saveGameExisted:bool, saveGame_Object:Save Game Object Reference) {is=title K2Node_FunctionEntry}
   > Branch(Condition=@n1)
   > n2 = Cast To TA_Base_SaveGame!(execute, Object=@n1.saveGame_Object) -> (then, CastFailed, AsTA Base Save Game:TA Base Save Game Object Reference) {is=title K2Node_DynamicCast}
   > Set UseSelectionCenter(UseSelectionCenter:bool="Get (a copy)"(Array:Array of Booleans=Get booleans!(self:TA Base Save Game Object Reference=@n2) -> (booleans:Array of Booleans)) -> (Output:bool)) -> (Output_Get:bool)
   n3 = Get vectors!(self:TA Base Save Game Object Reference=@n2) -> (vectors:Array of Vectors)
   > Set WorldInvertVector(WorldInvertVector:Vector="Get (a copy)"(Array:Array of Vectors=@n3) -> (Output:Vector)) -> (Output_Get:Vector)
   > Set WorldCenterIfNotUsingSelection(WorldCenterIfNotUsingSelection:Vector="Get (a copy)"(Array=@n3, Dimension 1=1)) -> (Output_Get:Vector)
   Return Node!(execute, " ":bool=false)
  ]}
 SaveGame_PrepareToSave {nodes [
   n1 = SaveGame_PrepareToSave!() -> (then, SaveGameObject:Save Game Object Reference) {is=title K2Node_FunctionEntry}
   > n2 = Cast To TA_Base_SaveGame(Object=@n1)
   > Set booleans!(
    execute
    booleans:Array of Booleans=Make Array("[0]":bool=Get UseSelectionCenter()) -> (Array:Array of Booleans)
    self:TA Base Save Game Object Reference=@n2
   ) -> (then, Output_Get:Array of Booleans)
   > Set vectors!(
    execute
    vectors:Array of Vectors=Make Array!("[0]":Vector=Get WorldInvertVector(), "[1]":Vector=Get WorldCenterIfNotUsingSelection()) -> (Array:Array of Vectors)
    self:TA Base Save Game Object Reference=@n2
   ) -> (then, Output_Get:Array of Vectors)
   Return Node(" "=false)
  ]}
]
variables [
 UseSelectionCenter {type=boolean}
 WorldCenterIfNotUsingSelection {type=Vector}
 WorldInvertVector {type=Vector, default=<1 0 0>}
]
widgets {
 info {
  parentClass=/Game/Blueprints/Blutilities/ForTechArt/Widgets/TA_Base_EditorWidget.TA_Base_EditorWidget_C
  rootWidgetClass=&StackBox
  widgetCount=3
 }
 widgets [
  StackBox_0 {widgetClassPath=&StackBox, widget=StackBox_0, properties {orientation=Orient_Vertical}}
  TA_Generic01_Button {
   widgetClassPath=&TA_Generic01_Button
   parent=StackBox_0
   widget=TA_Generic01_Button
   properties {label=Flip Selected Building Actors}
   slot {class=StackBoxSlot}
  }
  DetailsView_0 {
   widgetClassPath=&DetailsView
   parent=StackBox_0
   widget=DetailsView_0
   properties {categoriesToShow [nocategory]}
   slot {class=StackBoxSlot}
  }
 ]
}
